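# Additive glow flare material for the Rocket Pass item preview in the front end.
# A center-scaled, optionally rotating flare texture adds to switchable radial inner glow and ring gradients, then gets tinted and saturated.
# A switchable flicker multiplies the result, though both ends of its sine-driven lerp read FlickerRange red.
# A separate switchable branch lerps between top and bottom XYZ distances on a time sine for floating motion.
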Material /Game/Environments/FrontEnd/Blueprints/ItemPreview/RocketPass/Materials/M_Flare_RocketPass_Glow
graph {
 properties {blendMode=BLEND_Additive, pixelDepthOffsetMode=PDOM_Legacy, usedShadingModels=Default Lit}
 nodes [
  n1 = VectorParameter() {group=Flicker, parameterName=FlickerRange, defaultValue=<rgba 1 0.75 0 1>}
  n2 = Multiply(A=Fmod(A=Time(), B=Constant() {r=1}), B=ScalarParameter() {defaultValue=0.1, group=Flicker, parameterName=FlickerSpeed})
  n3 = Sine(None=Multiply(A=Sine(None=@n2)) {constB=0.5})
  n4 = ScalarParameter() {defaultValue=1, group=Texture, parameterName=Texture Scale}
  n5 = MaterialFunctionCall(UVs=TextureCoordinate(), Texture Scale=AppendVector(A=@n4, B=@n4)) {materialFunction=&ScaleUVsByCenter}
  n6 = Multiply(A=Time(), B=ScalarParameter() {defaultValue=2, group=Texture, parameterName=Texture RotationSpeed})
  n7 = TextureSampleParameter2D(UVs=StaticSwitchParameter(True=Rotator(Coordinate=@n5.UVs, Time=@n6), False=@n5.UVs) {defaultValue=true, group=Texture, parameterName=Is Rotating}) {
   group=Texture
   parameterName=Texture
   samplerSource=SSM_Clamp_WorldGroupSettings
   texture=/Game/Effects/Fort_Effects/Textures/Flares/T_Flare_Spikeball
  }
  n8 = Multiply(
   A=MaterialFunctionCall(
    Radius=ScalarParameter() {defaultValue=0.2, group=Glow, parameterName=Inner Glow Radius}
    Density=ScalarParameter() {defaultValue=0.5, group=Glow, parameterName=Inner Glow Hardness}
   ) {materialFunction=&RadialGradientExponential}.RadialGradientExponential
   B=ScalarParameter() {defaultValue=0.5, group=Glow, parameterName=Inner Glow Intensity}
  )
  n9 = Constant()
  n10 = Multiply(
   A=MaterialFunctionCall(
    Size=ScalarParameter() {defaultValue=0.5, group=Glow, parameterName=Glow Ring Size}
    GlowMin=ScalarParameter() {group=Glow, parameterName=Glow Ring Min}
    GlowMax=ScalarParameter() {defaultValue=0.3, group=Glow, parameterName=Glow Ring Max}
   ) {materialFunction=&MF_UI_Circle}.Outline
   B=ScalarParameter() {defaultValue=0.03, group=Glow, parameterName=Glow Ring Intensity, sliderMax=0.5}
  ) {constB=0.01}
  n11 = Saturate(None=Add(
    A=StaticSwitchParameter(True=@n8, False=@n9) {defaultValue=true, group=Glow, parameterName=Has Inner Glow, sortPriority=31}
    B=StaticSwitchParameter(True=@n10, False=@n9) {defaultValue=true, group=Glow, parameterName=Has Glow Ring, sortPriority=31}
   ))
  n12 = VectorParameter() {parameterName=Tint, defaultValue=<rgba 1 0.8978 0.3542 0>}
  n13 = Saturate(None=Multiply(A=Add(A=@n7.RGB, B=@n11), B=ComponentMask(None=@n12.RGB) {b=true, g=true, r=true}))
  StaticSwitchParameter(True=Multiply(A=LinearInterpolate(A=@n1.R, B=@n1.R, Alpha=@n3), B=@n13), False=@n13) {defaultValue=true, group=Flicker, parameterName=Flicker}
  n14 = Sine(None=Multiply(A=Time(), B=ScalarParameter() {defaultValue=0.1, group=Animation, parameterName=Floating Speed}))
  StaticSwitchParameter(
   True=LinearInterpolate(
    A=VectorParameter() {group=Animation, parameterName=Floating Top Distance XYZ, defaultValue=<rgba 0 0 1 0>}.RGB
    B=VectorParameter() {group=Animation, parameterName=Floating Bottom Distance XYZ}.RGB
    Alpha=NamedRerouteDeclaration(None=@n14) {name=Floating Lerp, nodeColor=<rgba 0 0 0 1>}
   )
   False=Constant()
  ) {defaultValue=true, group=Animation, parameterName=Is Floating, sortPriority=31}
 ]
}
