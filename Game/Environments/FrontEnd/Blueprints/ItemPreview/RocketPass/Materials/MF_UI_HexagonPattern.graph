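# Material function that draws a tiling hexagon pattern for UI, filed under the front-end Rocket Pass item preview.
# Scaled UVs are tested against two offset grids packed into one float4, and the nearest center yields a hexagonal distance field.
# Two feathered SmoothSteps, driven by hex distance and border thickness inputs, shape the masks.
# Outputs are outline, fill, outline border, and the raw and inverted distance field.
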MaterialFunction /Game/Environments/FrontEnd/Blueprints/ItemPreview/RocketPass/Materials/MF_UI_HexagonPattern
graph {nodes [
  n1 = NamedRerouteUsage()
  n2 = SmoothStep(Min=@n1, Max=Add(A=@n1, B=NamedRerouteUsage()), Value=NamedRerouteUsage()) {constMax=0.01, constMin=0.07}
  n3 = Add(A=NamedRerouteUsage(), B=NamedRerouteUsage())
  n4 = SmoothStep(Min=@n3, Max=Add(A=@n3, B=NamedRerouteUsage()), Value=NamedRerouteUsage()) {constMax=0.03}
  FunctionOutput(None=Saturate(None=Subtract(A=@n2, B=@n4) {constB=0.46})) {outputName=Outline Border}
  FunctionOutput(None=@n2) {outputName=Outline}
  FunctionOutput(None=OneMinus(None=@n2)) {outputName=Fill}
  n5 = NamedRerouteUsage()
  FunctionOutput(None=OneMinus(None=@n5)) {outputName=SDF outline}
  FunctionOutput(None=@n5) {outputName=SDF shape}
  n6 = Constant2Vector() {
   desc="Helper vector. If you're doing anything that involves regular triangles or hexagons, the\r\n30-60-90 triangle will be involved in some way, which has sides of 1, sqrt(3) and 2"
   g=1
   r=1.7321
  }
  NamedRerouteDeclaration(None=@n6) {name=sqRect, nodeColor=<rgba 0 0 0 1>}
  n7 = FunctionInput(Preview=TextureCoordinate()) {bUsePreviewValueAsDefault=true, inputName=UVs, inputType=FunctionInput_Vector2}
  n8 = NamedRerouteUsage()
  n9 = NamedRerouteDeclaration(None=Add(A=Multiply(A=@n7, B=@n8) {constB=5}, B=NamedRerouteUsage())) {name=uvs, nodeColor=<rgba 0 0 0 1>}
  n10 = AppendVector(A=@n9, B=Subtract(A=@n9, B=Constant2Vector() {g=0.5, r=1}))
  n11 = NamedRerouteUsage()
  n12 = AppendVector(A=@n11, B=@n11)
  n13 = Add(A=Floor(None=Divide(A=@n10, B=@n12) {desc=grouping some float2 calculations in a float4 calculation})) {constB=0.5}
  n14 = Subtract(A=NamedRerouteUsage(), B=Multiply(A=ComponentMask(None=@n13) {g=true, r=true}, B=NamedRerouteUsage())) {desc=Hexagon Centers Even}
  n15 = ComponentMask(None=@n13) {a=true, b=true}
  n16 = Subtract(A=NamedRerouteUsage(), B=Multiply(A=Add(A=@n15) {constB=0.5}, B=NamedRerouteUsage())) {desc=Hexagon Centers Odd}
  n17 = Abs(None=If(A=DotProduct(A=@n14, B=@n14), B=DotProduct(A=@n16, B=@n16), "A > B"=@n16, "A == B"=@n16, "A < B"=@n14) {desc="Nearest hexagon center to the current point.\r\n(When value is zero, we're at the center)"})
  n18 = DotProduct(A=@n17, B=Multiply(A=NamedRerouteUsage()) {constB=0.5})
  n19 = ComponentMask(None=@n17) {g=true}
  NamedRerouteDeclaration(None=Multiply(A=Max(A=@n18, B=@n19)) {constB=2}) {name=SDF Shape, nodeColor=<rgba 1 0 0.8203 1>}
  NamedRerouteDeclaration(None=FunctionInput(Preview=Constant() {r=0.8}) {bUsePreviewValueAsDefault=true, inputName=Hex Distance, inputType=FunctionInput_Scalar}) {name=hex distance, nodeColor=<rgba 0 1 0.8359 1>}
  NamedRerouteDeclaration(None=@n13) {name=combined hex centers, nodeColor=<rgba 0 0 0 1>}
  NamedRerouteDeclaration(None=FunctionInput(Preview=Constant() {r=0.05}) {bUsePreviewValueAsDefault=true, inputName=Hex Border Thickness, inputType=FunctionInput_Scalar}) {name=hex thickness, nodeColor=<rgba 0 0.8594 1 1>}
  NamedRerouteDeclaration(None=ScalarParameter() {defaultValue=0.03, parameterName=Feather}) {name=feather, nodeColor=<rgba 1 0 0.8437 1>}
  NamedRerouteDeclaration(None=FunctionInput(Preview=ScalarParameter() {defaultValue=4, parameterName=Tiling}) {bUsePreviewValueAsDefault=true, inputName=Tiling, inputType=FunctionInput_Scalar, sortPriority=1}) {name=tiling, nodeColor=<rgba 0.5 0 1 1>}
 ]}
