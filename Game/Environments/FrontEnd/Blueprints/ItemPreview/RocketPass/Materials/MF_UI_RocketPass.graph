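# Material function that composes the Rocket Pass front-end item preview visual from a stylized background and a channel-packed item image.
# A rotated five-color gradient carries a panning hexagon pattern with noise-driven highlights and an optional shadow from an offset second hex evaluation.
# Texture channels act as masks for tinted image, gradient shadow and optional outline, while an offset resample adds a switchable drop shadow.
# The Floating Lerp input blends in parallax offsets for image and shadow.
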MaterialFunction /Game/Environments/FrontEnd/Blueprints/ItemPreview/RocketPass/Materials/MF_UI_RocketPass
graph {nodes [
  n1 = NamedRerouteUsage()
  n2 = Saturate(None=ConstantBiasScale(None=ComponentMask(None=@n1) {g=true}) {bias=-0.25, scale=2}) {desc=Vertical Gradient}
  n3 = MaterialFunctionCall(UVs=NamedRerouteUsage(), "Rotation Angle (0-1)"=Constant() {r=-0.125}) {materialFunction=&CustomRotator}
  n4 = MaterialFunctionCall!(
   In=Saturate(None=ComponentMask(None=@n3.Rotated Values) {g=true})
   Color_1=VectorParameter() {group=Background, parameterName=BG Gradient Color 1, defaultValue=<rgba 1 0.6496 0.1406 1>}.RGB
   Color_2=VectorParameter() {group=Background, parameterName=BG Gradient Color 2, defaultValue=<rgba 1 0.3231 0.003 1>}.RGB
   Color_3=VectorParameter() {group=Background, parameterName=BG Gradient Color 3, defaultValue=<rgba 0.2016 0.1912 0.2346 1>}.RGB
   Color_4=VectorParameter() {group=Background, parameterName=BG Gradient Color 4, defaultValue=<rgba 0 0.0284 0.956 1>}.RGB
   Color_5=VectorParameter() {group=Background, parameterName=BG Gradient Color 5, defaultValue=<rgba 0.0273 0.2986 1 1>}.RGB
   Pos Color_1=ScalarParameter() {group=Background, parameterName=Pos Color 1}
   Pos Color_2=ScalarParameter() {defaultValue=0.2, group=Background, parameterName=Pos Color 2}
   Pos Color_3=ScalarParameter() {defaultValue=0.5, group=Background, parameterName=Pos Color 3}
   Pos Color_4=ScalarParameter() {defaultValue=0.8, group=Background, parameterName=Pos Color 4}
   Pos Color_5=ScalarParameter() {defaultValue=1, group=Background, parameterName=Pos Color 5}
  ) {materialFunction=&/MF_UI_RemappableGradient}
  n5 = Panner(
   Coordinate=NamedRerouteUsage()
   Speed=VectorParameter() {group=Hexagon Pattern, parameterName=Hex Highlight Panning, defaultValue=<rgba 0.1 0.2 0 0>}.RGB
  ) {speedX=0.1, speedY=0.2}
  n6 = ScalarParameter() {defaultValue=0.01, group=Hexagon Pattern, parameterName=HexThickness, sliderMax=1}
  n7 = ScalarParameter() {defaultValue=0.9, group=Hexagon Pattern, parameterName=HexSize, sliderMax=1}
  n8 = ScalarParameter() {defaultValue=-0.01, group=Hexagon Pattern, parameterName=Hex Panning X, sliderMax=0.2, sliderMin=-0.2}
  n9 = ScalarParameter() {defaultValue=0.02, group=Hexagon Pattern, parameterName=Hex Panning Y, sliderMax=0.2, sliderMin=-0.2}
  n10 = Add(A=Panner(Coordinate=NamedRerouteUsage(), Speed=AppendVector(A=@n8, B=@n9)) {speedX=0.1, speedY=0.2}) {constB=0.1}
  n11 = ScalarParameter() {defaultValue=5.5, group=Hexagon Pattern, parameterName=HexTiling}
  n12 = MaterialFunctionCall!(Hex Border Thickness=@n6, Hex Distance=@n7, UVs=@n10, Tiling=@n11) {materialFunction=&/MF_UI_HexagonPattern}
  n13 = Multiply(
   A=Multiply(
    A=TextureSample(UVs=@n5) {samplerType=SAMPLERTYPE_Grayscale, texture=&LowResBlurredNoise}.R
    B=ScalarParameter() {defaultValue=1, group=Hexagon Pattern, parameterName=Hex Highlight Intensity}
   )
   B=@n12.Outline Border
  )
  n14 = ScalarParameter() {defaultValue=0.15, group=Hexagon Pattern, parameterName=Hex Outline Intensity}
  n15 = Multiply(A=Saturate(None=Add(A=@n4.Gradient 5 Colors, B=@n13)), B=OneMinus(None=Multiply(A=@n12.Outline, B=@n14)))
  n16 = Constant() {r=-0.005}
  n17 = Constant() {r=-0.005}
  n18 = MaterialFunctionCall(Hex Border Thickness=@n6, Hex Distance=@n7, UVs=Add(A=@n10, B=AppendVector(A=@n16, B=@n17)) {constB=0.01}, Tiling=@n11) {materialFunction=&MF_UI_HexagonPattern}
  n19 = Multiply(
   A=Saturate(None=Subtract(A=@n12.Outline, B=@n18.Outline))
   B=ScalarParameter() {defaultValue=0.35, group=Hexagon Pattern, parameterName=Hex Shadow Intensity, sliderMax=1}
  ) {constB=0.2}
  n20 = StaticSwitchParameter(True=Multiply(A=@n15, B=OneMinus(None=@n19)), False=@n15) {defaultValue=true, group=Hexagon Pattern, parameterName=Show Hex Shadow}
  n21 = LinearInterpolate(
   A=@n20
   B=LinearInterpolate(
    A=VectorParameter() {group=Image, parameterName=Shadow Outline Gradient Color 1, defaultValue=<rgba 0.7305 0.117 0.0742 1>}.RGB
    B=VectorParameter() {group=Image, parameterName=Shadow Outline Gradient Color 2, defaultValue=<rgba 0.0052 0.1779 0.5395 1>}.RGB
    Alpha=@n2
   ) {desc=Shadow Outline}
   Alpha=NamedRerouteUsage()
  )
  n22 = VectorParameter() {group=Image, parameterName=Shadow Gradient Color 1, defaultValue=<rgba 0.4233 0.0513 0.0144 1>}
  n23 = VectorParameter() {group=Image, parameterName=Shadow Gradient Color 2, defaultValue=<rgba 0 0.0152 0.2582 1>}
  n24 = LinearInterpolate(
   A=LinearInterpolate(
    A=StaticSwitchParameter(True=@n21, False=@n20) {defaultValue=true, group=Image, parameterName=Shadow Outline}
    B=LinearInterpolate(A=@n22.RGB, B=@n23.RGB, Alpha=@n2) {desc=Colored Shadow}
    Alpha=NamedRerouteUsage()
   )
   B=VectorParameter() {group=Image, parameterName=Image Color, defaultValue=<rgba 1 1 1 1>}.RGB
   Alpha=NamedRerouteUsage()
  )
  n25 = MaterialFunctionCall(
   UVs=NamedRerouteUsage()
   Texture Scale=AppendVector(
    A=ScalarParameter() {defaultValue=0.8, group=Image, parameterName=Texture Scale X}
    B=ScalarParameter() {defaultValue=0.8, group=Image, parameterName=Texture Scale Y}
   )
  ) {materialFunction=&ScaleUVsByCenter}
  n26 = ScalarParameter() {group=Image, parameterName=Image Floating Parallax X}
  n27 = ScalarParameter() {defaultValue=-0.02, group=Image, parameterName=Image Floating Parallax Y}
  n28 = LinearInterpolate(A=@n25.UVs, B=Add(A=@n25.UVs, B=AppendVector(A=@n26, B=@n27)), Alpha=NamedRerouteUsage())
  n29 = Add(
   A=@n28
   B=AppendVector(
    A=ScalarParameter() {defaultValue=-0.01, group=Image, parameterName=Shadow Offset X, sliderMax=0.2, sliderMin=-0.2}
    B=ScalarParameter() {defaultValue=-0.02, group=Image, parameterName=Shadow Offset Y, sliderMax=0.2, sliderMin=-0.2}
   )
  )
  n30 = AppendVector(A=ScalarParameter() {parameterName=Image Shadow Parallax X}, B=ScalarParameter() {defaultValue=0.02, parameterName=Image Shadow Parallax Y})
  n31 = TextureSample(UVs=LinearInterpolate(A=@n29, B=Add(A=@n29, B=@n30), Alpha=NamedRerouteUsage()), Tex=NamedRerouteUsage())
  n32 = Max(A=NamedRerouteUsage(), B=NamedRerouteUsage())
  n33 = NamedRerouteUsage()
  n34 = Subtract(
   A=Max(A=Max(A=@n31.R, B=@n31.R), B=@n31.B)
   B=StaticSwitchParameter(True=Max(A=@n32, B=@n33), False=@n32) {defaultValue=true, group=Image, parameterName=Shadow Outline}
  )
  n35 = ScalarParameter() {defaultValue=0.6, group=Image, parameterName=Shadow Intensity, sliderMax=1}
  n36 = Saturate(None=OneMinus(None=Multiply(A=@n34, B=@n35) {constB=0.1}))
  FunctionOutput(None=StaticSwitchParameter(True=Multiply(A=@n24, B=@n36) {constA=1}, False=@n24) {defaultValue=true, group=Image, parameterName=Drop Shadow})
  NamedRerouteDeclaration(None=FunctionInput(Preview=Constant()) {bUsePreviewValueAsDefault=true, desc=Is the image animating?, inputName=Floating Lerp, inputType=FunctionInput_Scalar}) {name=floating lerp, nodeColor=<rgba 0 0 0 1>}
  n37 = TextureObjectParameter() {
   group=Image
   parameterName=Item Image
   samplerType=SAMPLERTYPE_LinearColor
   texture=/Game/Environments/FrontEnd/Blueprints/ItemPreview/RocketPass/Textures/T_Icon_RocketPass_Logo_CP
  }
  NamedRerouteDeclaration(None=@n37) {name=Item Image Texture, nodeColor=<rgba 0 0 0 1>}
  n38 = ScalarParameter() {group=Image, parameterName=Texture Offset X, sliderMax=0.2, sliderMin=-0.2}
  n39 = ScalarParameter() {group=Image, parameterName=Texture Offset Y, sliderMax=0.2, sliderMin=-0.2}
  n40 = TextureSample(UVs=Add(A=AppendVector(A=@n38, B=@n39), B=@n28), Tex=NamedRerouteUsage())
  NamedRerouteDeclaration(None=@n40.R) {name=R Channel, nodeColor=<rgba 0 0 0 1>}
  NamedRerouteDeclaration(None=@n40.G) {name=G Channel, nodeColor=<rgba 0 0 0 1>}
  NamedRerouteDeclaration(None=@n40.B) {name=B Channel, nodeColor=<rgba 0 0 0 1>}
  NamedRerouteDeclaration(None=FunctionInput(Preview=TextureCoordinate()) {bUsePreviewValueAsDefault=true, inputName=UVs, inputType=FunctionInput_Vector2}) {name=uv, nodeColor=<rgba 0 0 0 1>}
  TextureCoordinate()
 ]}
