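# Material function in the Dojo pickaxe FX folder that builds a movable sheen band mask along a scalar gradient, defaulting to horizontal UV.
# Mask Position sweeps a Mask Size window from fully before to fully past the gradient, previewing with a debug time sine.
# The remapped, saturated window outputs as a Linear ramp, a Sinusoidal soft peak, and a Binary band via Ceil of Frac.
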MaterialFunction /Game/Weapons/FORT_Melee/Pickaxe_Dojo_Male_1H/FX/MF_Sheen
graph {nodes [
  n1 = TextureCoordinate()
  n2 = Saturate(None=FunctionInput(Preview=ComponentMask(None=@n1) {r=true}) {bUsePreviewValueAsDefault=true, description=Clamped 0-1, inputName=Gradient, inputType=FunctionInput_Scalar})
  n3 = FunctionInput(Preview=MaterialFunctionCall() {materialFunction=&DebugTimeSine}.Result) {bUsePreviewValueAsDefault=true, inputName=Mask Position, inputType=FunctionInput_Scalar, sortPriority=1}
  n4 = FunctionInput() {bUsePreviewValueAsDefault=true, inputName=Mask Size, inputType=FunctionInput_Scalar, sortPriority=2, previewValue=<0.333 0 0 1>}
  n5 = Subtract(A=Multiply(A=Saturate(None=@n3), B=Add(A=@n4)), B=@n4)
  n6 = Saturate(None=Divide(A=Subtract(A=@n2, B=@n5), B=@n4)) {desc=Remapped gradient}
  FunctionOutput(None=@n6) {description=0-1 Binary Mask, outputName=Linear}
  n7 = Subtract(A=@n6) {constB=0.25}
  FunctionOutput(None=ConstantBiasScale(None=Sine(None=@n7))) {description=0-1 Sine Mask, outputName=Sinusoidal, sortPriority=2}
  FunctionOutput(None=Ceil(None=Frac(None=@n6))) {description=0-1 Clamped Linear Mask, outputName=Binary, sortPriority=1}
 ]}
