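# Unlit, two-sided material for the Rocket Pass in the front-end item preview, showing card art inside an animated gradient border.
# A four-color gradient plus a sheen looping every two seconds forms the border and entire back face.
# A UV-distance mask limits contrast- and brightness-adjusted MF_UI_RocketPass art to the front interior.
# A time-driven sine feeds that art function and a switchable floating offset between two vectors.
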Material /Game/Environments/FrontEnd/Blueprints/ItemPreview/RocketPass/Materials/M_RocketPass
graph {
 properties {
  bUsedWithMeshDeformer=true
  bUsedWithSkeletalMesh=true
  pixelDepthOffsetMode=PDOM_Legacy
  shadingModel=MSM_Unlit
  twoSided=true
  usedShadingModels=Unlit
 }
 nodes [
  n1 = MaterialFunctionCall!(
   In=ComponentMask(None=TextureCoordinate()) {g=true}
   Color_1=VectorParameter() {group=Border Gradient, parameterName=GradientColor1, defaultValue=<rgba 0.9323 0.6034 0.0631 1>}.RGB
   Color_2=VectorParameter() {group=Border Gradient, parameterName=GradientColor2, defaultValue=<rgba 0.8879 0.6884 0.3607 1>}.RGB
   Color_3=VectorParameter() {group=Border Gradient, parameterName=GradientColor3, defaultValue=<rgba 0.7991 0.3832 0.0166 1>}.RGB
   Color_4=VectorParameter() {group=Border Gradient, parameterName=GradientColor4, defaultValue=<rgba 0.9823 0.7575 0.1944 1>}.RGB
   Color_5
   Pos Color_1=ScalarParameter() {group=Border Gradient, parameterName=PosColor1}
   Pos Color_2=ScalarParameter() {defaultValue=0.3, group=Border Gradient, parameterName=PosColor2}
   Pos Color_3=ScalarParameter() {defaultValue=0.6, group=Border Gradient, parameterName=PosColor3}
   Pos Color_4=ScalarParameter() {defaultValue=1, group=Border Gradient, parameterName=PosColor4}
   Pos Color_5
  ) {materialFunction=&/MF_UI_RemappableGradient}
  n2 = Rotator(Coordinate=TextureCoordinate(), Time=ScalarParameter() {defaultValue=-3.14, group=Border Gradient, parameterName=Sheen Rotation})
  n3 = Fmod(A=Time(), B=Constant() {r=2})
  n4 = Multiply(
   A=MaterialFunctionCall!(Gradient=ComponentMask(None=@n2) {g=true}, Mask Position=OneMinus(None=@n3), Mask Size) {materialFunction=/Game/Weapons/FORT_Melee/Pickaxe_Dojo_Male_1H/FX/MF_Sheen}.Sinusoidal
   B=ScalarParameter() {defaultValue=0.8, group=Border Gradient, parameterName=Sheen Intensity}
  )
  n5 = Power(
   Base=MaterialFunctionCall!(UVs, Floating Lerp=NamedRerouteUsage()) {materialFunction=&/MF_UI_RocketPass}.Result
   Exp=ScalarParameter() {defaultValue=1.5, parameterName=Contrast}
  )
  n6 = ScalarParameter() {defaultValue=1.2, parameterName=Brightness}
  n7 = MaterialFunctionCall(Float2=TextureCoordinate()) {materialFunction=&BreakOutFloat2Components}
  n8 = Constant() {r=0.5}
  n9 = Distance(A=@n7.R, B=@n8)
  n10 = Distance(A=@n7.G, B=@n8)
  n11 = Multiply(
   A=MaterialFunctionCall(Alpha=Max(A=@n9, B=@n10), Min=Constant() {r=0.455}, Max=Constant() {r=0.45}) {materialFunction=&SmoothStep}.Result
   B=Saturate(None=TwoSidedSign())
  )
  Multiply(
   A=LinearInterpolate(A=Add(A=@n1.Gradient 4 Colors, B=@n4), B=Multiply(A=@n5, B=@n6), Alpha=Saturate(None=@n11))
   B=ScalarParameter() {defaultValue=1.6, parameterName=Emissive Intensity}
  )
  n12 = Sine(None=Multiply(A=Time(), B=ScalarParameter() {defaultValue=0.1, group=Animation, parameterName=Floating Speed}))
  StaticSwitchParameter(
   True=LinearInterpolate(
    A=VectorParameter() {group=Animation, parameterName=Floating Top Distance XYZ, defaultValue=<rgba 0 0 5 0>}.RGB
    B=VectorParameter() {group=Animation, parameterName=Floating Bottom Distance XYZ}.RGB
    Alpha=NamedRerouteDeclaration(None=@n12) {name=floating lerp, nodeColor=<rgba 0 0 0 1>}
   )
   False=Constant()
  ) {defaultValue=true, group=Animation, parameterName=Is Floating, sortPriority=31}
 ]
}
